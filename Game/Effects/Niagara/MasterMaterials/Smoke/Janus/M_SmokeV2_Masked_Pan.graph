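# Master smoke material for Niagara sprite and mesh particles, translucent despite its Masked name, with vertex-normal displacement scaled by opacity.
# A panning, per-particle randomized noise texture erodes particle alpha, which is then powered and remapped for opacity.
# Color blends parameter or particle color, with time-of-day tint and sun-direction shading in a switchable fake-lit emissive path.
# Optional HDR emissive uses a curve atlas, hue shift and detail noise, tuned through dynamic parameters.
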Material /Game/Effects/Niagara/MasterMaterials/Smoke/Janus/M_SmokeV2_Masked_Pan
graph {
 properties {
  bCastRayTracedShadows=false
  bHasPixelAnimation=true
  bUsedWithNanite=true
  bUsedWithNiagaraMeshParticles=true
  bUsedWithNiagaraSprites=true
  blendMode=BLEND_Translucent
  pixelDepthOffsetMode=PDOM_Legacy
  translucencyLightingMode=TLM_VolumetricPerVertexDirectional
  translucentSelfShadowDensityScale=1
  usedShadingModels=Default Lit
  translucentMultipleScatteringExtinction=<rgba 1 1 1 1>
 }
 nodes [
  n1 = VertexNormalWS()
  MaterialFunctionCall(
   BaseNormal=Transform(None=@n1) {transformSourceType=TRANSFORMSOURCE_World, transformType=TRANSFORM_Tangent}
   AdditionalNormal=NamedRerouteUsage()
  ) {materialFunction=&BlendAngleCorrectedNormals}
  Multiply(A=ScalarParameter() {defaultValue=1, parameterName=ScaleColor, sliderMax=2}, B=NamedRerouteUsage())
  n2 = Power(
   Base=Multiply(A=ScalarParameter() {defaultValue=1, parameterName=ScaleOpacity, sliderMax=2}, B=NamedRerouteUsage())
   Exp=ScalarParameter() {defaultValue=1.5, parameterName=MaskPower, sliderMax=5}
  )
  n3 = Saturate(None=MaterialFunctionCall(
    Input=Saturate(None=@n2)
    Input Low=ScalarParameter() {parameterName=Opacity Floor}
    Input High=ScalarParameter() {defaultValue=0.4, parameterName=Opacity Ceil, sliderMax=1}
    Target Low=Constant()
    Target High=Constant() {r=1}
   ) {materialFunction=&RemapValueRange}.Result)
  Multiply(
   A=MaterialFunctionCall(Alpha Threshold=ScalarParameter() {defaultValue=0.6, parameterName=DitherAlphaThresh, sliderMax=5}, Random=ParticleRandom()) {materialFunction=&DitherTemporalAA}.Result
   B=@n3
  )
  n4 = ScalarParameter() {defaultValue=3, parameterName=Pan U}
  n5 = DynamicParameter() {
   parameterIndex=2
   defaultValue=<rgba 0 0 1 1>
   paramNames [Smoke - DissolveLerpNoise, Smoke - DissolveNoiseOffsetX, Smoke - DissolveNoiseErosionPower, Smoke - DissolveNoiseScale]
  }
  n6 = Add(
   A=MaterialFunctionCall!(
    UV=TextureCoordinate()
    Tile U=ScalarParameter() {defaultValue=1, parameterName=U Tile, sliderMax=1}
    Tile V=ScalarParameter() {defaultValue=1, parameterName=V Tile, sliderMax=1}
    Pan U=Multiply(A=@n4) {constB=-1}
    Pan V
    Offset U
    Offset V
    Time
   ) {materialFunction=/Game/Effects/Fort_Effects/Materials/_Functions/MF_PanTileOffset}.UV's
   B=@n5.Smoke - DissolveNoiseOffsetX
  )
  n7 = Multiply(
   A=ParticleRandom()
   B=AppendVector(A=Constant(), B=ScalarParameter() {group=Normal, parameterName="DissolveDetail  Random Y Offset Value", sliderMax=2})
  )
  n8 = TextureSampleParameter2D(UVs=Multiply(A=Add(A=@n6, B=@n7), B=@n5.Smoke - DissolveNoiseScale) {desc=Scale}) {parameterName=DissolveDetailTexture, samplerType=SAMPLERTYPE_LinearColor, texture=/Game/Global/Textures/Default/Noise/T_Noise01_Inverse}
  n9 = DynamicParameter() {
   parameterIndex=2
   defaultValue=<rgba 0 0 1 1>
   paramNames [Smoke - DissolveLerpNoise, Smoke - DissolveNoiseOffsetX, Smoke - DissolveNoiseErosionPower, Smoke - DissolveNoiseScale]
  }
  NamedRerouteDeclaration(None=Clamp(None=Power(Base=@n8.G, Exp=@n9.Smoke - DissolveNoiseErosionPower))) {name=DissolveDetail, nodeColor=<rgba 0.3828 0 1 1>}
  n10 = NamedRerouteUsage()
  n11 = Multiply(A=ParticleColor().A, B=@n10)
  n12 = DynamicParameter() {
   parameterIndex=2
   defaultValue=<rgba 0 0 1 1>
   paramNames [Smoke - DissolveLerpNoise, Smoke - DissolveNoiseOffsetX, Smoke - DissolveNoiseErosionPower, Smoke - DissolveNoiseScale]
  }
  NamedRerouteDeclaration(None=LinearInterpolate(A=@n11, B=Multiply(A=@n11, B=@n10), Alpha=Clamp(None=@n12.Smoke - DissolveLerpNoise))) {name=Final Alpha, nodeColor=<rgba 1 0 0.2578 1>}
  n13 = VectorParameter() {parameterName=Color, defaultValue=<rgba 0.3958 0.3987 0.4 0>}
  n14 = ParticleColor()
  n15 = ScalarParameter() {parameterName=UseParticleColor}
  n16 = Multiply(
   A=MaterialFunctionCall(
    Color_Morning=VectorParameter() {parameterName=Color Morning Multiplier, defaultValue=<rgba 0.9896 0.268 0.4929 0>}.RGB
    Color_Day=VectorParameter() {parameterName=Color Day Multiplier, defaultValue=<rgba 1 1 1 0>}.RGB
    Color_Evening=VectorParameter() {parameterName=Color Evening Multiplier, defaultValue=<rgba 0.599 0.2901 0.3358 0>}.RGB
    Color_Night=VectorParameter() {parameterName=Color Night Multiplier, defaultValue=<rgba 0.3498 0.4864 0.6458 0>}.RGB
   ) {materialFunction=&Color_Time_of_Day}.Result
   B=NamedRerouteDeclaration(None=LinearInterpolate(A=@n13.RGB, B=@n14.RGB, Alpha=@n15)) {name=Final_Base Color, nodeColor=<rgba 1 0.7266 0 1>}
  )
  n17 = NamedRerouteUsage()
  n18 = CollectionParameter() {
   collection=&FortniteMaterialParameters
   parameterName=SunlightDirectionalVector
  }
  n19 = LinearInterpolate(
   A=Multiply(A=@n16, B=ScalarParameter() {defaultValue=0.7, parameterName=Shadow Multiplier}) {constB=0.74}
   B=@n16
   Alpha=DotProduct(A=@n17, B=ComponentMask(None=@n18) {b=true, g=true, r=true})
  )
  n20 = Add(A=Multiply(A=@n13.RGB, B=ScalarParameter() {parameterName=Emissive}), B=NamedRerouteUsage())
  NamedRerouteDeclaration(None=StaticSwitchParameter(True=Add(A=@n19, B=@n20), False=@n20) {parameterName=IsShaderUnlit}) {name=Final_Emissive, nodeColor=<rgba 0 1 0.7656 1>}
  n21 = Constant3Vector() {constant=<rgba 0 0 1 0>}
  n22 = NamedRerouteDeclaration(None=TextureSample() {samplerType=SAMPLERTYPE_Normal, texture=/Engine/Functions/Engine_MaterialFunctions02/ExampleContent/Textures/SphereRenderNormalsMap}.RGB) {name=SphereNormal_Texture, nodeColor=<rgba 1 0.3281 0 1>}
  n23 = Multiply(
   A=ScalarParameter() {defaultValue=1, group=Normal, parameterName=SphereNormal}
   B=DynamicParameter() {paramNames [Animation, Unused, Emissive, NormalMultiplier]}.NormalMultiplier
  )
  NamedRerouteDeclaration(None=Normalize(VectorInput=LinearInterpolate(A=@n21, B=@n22, Alpha=@n23))) {name=SphereNormal, nodeColor=<rgba 1 0 0.2109 1>}
  NamedRerouteDeclaration(None=@n1) {name=Final Normal, nodeColor=<rgba 1 0 0.75 1>}
  n24 = CurveAtlasRowParameter(CurveTime=NamedRerouteUsage()) {
   atlas=/Game/Effects/Niagara/MasterMaterials/Smoke/Textures/CurveAtlas
   curve=/Game/Effects/Niagara/MasterMaterials/Smoke/Textures/EmissiveCurve
   defaultValue=1
   parameterName=EmissiveCurve
  }
  n25 = Multiply(A=NamedRerouteUsage(), B=@n24)
  n26 = MaterialFunctionCall(
   Hue Shift Percentage=ScalarParameter() {parameterName=HueShift}
   Texture=StaticSwitchParameter(True=Clamp(None=@n25), False=@n24) {parameterName=Multiply HDR Emissive by Base Color}
  ) {materialFunction=&HueShift}
  n27 = MaterialFunctionCall(
   LDR Input=@n26.Result
   HDR Tint=@n26.Result
   HDR Intensity Multiplier=Multiply(
    A=ScalarParameter() {defaultValue=100, parameterName=HDR_Intensity}
    B=DynamicParameter() {paramNames [Animation, Unused, Emissive, NormalMultiplier]}.Emissive
   )
  ) {materialFunction=&DeriveHDRfromLDR}
  n28 = NamedRerouteUsage()
  n29 = Add(A=Multiply(A=ComponentMask(None=@n28) {g=true, r=true}) {constB=-0.2}, B=TextureCoordinate())
  n30 = DynamicParameter() {
   parameterIndex=1
   defaultValue=<rgba 0 0 1 1>
   paramNames [Emissive - LerpNoise, Emissive - NoiseOffsetX, Emissive - NoiseErosionPower, Emissive - NoiseScale]
  }
  n31 = Multiply(
   A=@n27.RemappedRange
   B=Power(
    Base=TextureSampleParameter2D(UVs=@n29) {parameterName=DetailTexture, texture=/Game/Global/Textures/Default/Tileable_Patterns/T_Blowing_Smoke_Pattern2_Tile_DISP}.G
    Exp=@n30.Emissive - NoiseErosionPower
   )
  )
  n32 = DynamicParameter() {
   parameterIndex=1
   defaultValue=<rgba 0 0 1 1>
   paramNames [Emissive - LerpNoise, Emissive - NoiseOffsetX, Emissive - NoiseErosionPower, Emissive - NoiseScale]
  }
  n33 = Clamp(None=@n32.Emissive - LerpNoise)
  NamedRerouteDeclaration(None=StaticSwitchParameter(True=LinearInterpolate(A=@n27.RemappedRange, B=@n31, Alpha=@n33), False=Constant()) {parameterName=UseEmissive}) {name=HDR Emissive, nodeColor=<rgba 1 0 0.3047 1>}
  NamedRerouteUsage()
  n34 = VertexNormalWS()
  n35 = Multiply(A=@n3, B=ScalarParameter() {defaultValue=5, parameterName=Displacement, sliderMax=25})
  Reroute(None=Multiply(A=@n34, B=Max(A=@n35) {constB=0}))
  Reroute(None=@n34)
  Reroute(None=@n3)
  Transform(None=@n17)
 ]
}
